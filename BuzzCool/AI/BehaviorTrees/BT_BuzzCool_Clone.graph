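# Behavior tree for BuzzCool clone NPCs, choosing movement and wandering by goal actor and hostility.
# Hostile goals trigger a melee token query: token holders close in via evaluator positions or fallback moves, pausing if the goal is airborne; others run hostile wander or idle-break abilities.
# Non-hostile goals alternate cooldown-gated idling with EQS moves near the item, tagged as skipping.
# Without a goal, clones loop a free-wander ability.
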BehaviorTree /BuzzCool/AI/BehaviorTrees/BT_BuzzCool_Clone
blackboard=&/BB_BuzzCool_Clone
root {
 name=Selector
 class=BTComposite_Selector
 children [
  Selector {
   class=BTComposite_Selector
   children [
    Selector {
     class=BTComposite_Selector
     children [
      Sequence {
       class=BTComposite_Sequence
       children [
        Wait for Goal Pawn to Fall Off {
         class=BTTask_Wait
         decorators [
          Can Actor be considered Airborne {
           class=FortBTDecorator_CanBeConsideredAirborne
           properties {
            bInverseCondition=false
            flowAbortMode=None
            nodeName=Can Actor be considered Airborne
            blackboardKey {bNoneIsAllowedValue=false, selectedKeyName=GoalActor}
           }
          }
          Force Success {class=BTDecorator_ForceSuccess}
         ]
         properties {nodeName=Wait for Goal Pawn to Fall Off, randomDeviation {defaultValue=0.1}, waitTime {defaultValue=0.25}}
        }
        Selector {
         class=BTComposite_Selector
         children [
          Selector {
           class=BTComposite_Selector
           children [
            Move to Token Position {
             class=FortBTTask_GameMoveTo
             decorators [Distance Between {
               class=FortBTDecorator_DistanceBetween
               properties {
                flowAbortMode=Both
                nodeName=Distance Between
                operator=LessOrEqual
                blackboardKeyB {selectedKeyName=GoalActor}
                specifiedDistance {defaultValue=100}
               }
              }]
             properties {
              bDeimosFlavor=false
              bFinishMoveOnOverlap=false
              bStopAtGoal=true
              bUpdateMoveOnBlackboardKeyChanges=true
              nodeName=Move to Token Position
              acceptableRadius {defaultValue=20}
              bReachTestIncludesAgentRadius {defaultValue=false}
              bReachTestIncludesGoalRadius {defaultValue=false}
              blackboardKey {selectedKeyName=AIEvaluator_TokenPosition_Destination}
              filterClass {defaultValue=/Game/Athena/AI/NavFilters/NavFilter_NPC_ExcludeWater.NavFilter_NPC_ExcludeWater_C}
              navFilterTag {tagName=None}
             }
            }
            Move to Token Position {
             class=FortBTTask_GameMoveTo
             properties {
              bDeimosFlavor=true
              bFinishMoveOnOverlap=false
              bStopAtGoal=true
              bUpdateMoveOnBlackboardKeyChanges=true
              nodeName=Move to Token Position
              acceptableRadius {defaultValue=20}
              bReachTestIncludesAgentRadius {defaultValue=false}
              bReachTestIncludesGoalRadius {defaultValue=false}
              blackboardKey {selectedKeyName=AIEvaluator_TokenPosition_Destination}
              filterClass {defaultValue=&NavFilter_NPC_ExcludeWater_C}
              navFilterTag {tagName=None}
             }
            }
           ]
           decorators [
            Have a Token Position ? {
             class=BTDecorator_Blackboard
             properties {
              arithmeticOperation=GreaterOrEqual
              intValue=3
              nodeName=Have a Token Position ?
              stringValue=ExecutionPending
              blackboardKey {selectedKeyName=AIEvaluator_TokenPosition_ExecutionStatus}
             }
            }
            Distance Between {
             class=FortBTDecorator_DistanceBetween
             properties {
              flowAbortMode=Both
              nodeName=Distance Between
              operator=Greater
              blackboardKeyB {selectedKeyName=AIEvaluator_TokenPosition_Destination}
              specifiedDistance {defaultValue=100}
             }
            }
           ]
           properties {
            children [
             {
              childComposite=null
              childTask=&self:FortBTTask_GameMoveTo_2
              decorators [&self:FortBTDecorator_DistanceBetween_2]
             }
             {childComposite=null, childTask=&self:FortBTTask_GameMoveTo_0, decorators []}
            ]
            services []
           }
           services []
          }
          Fallback Move to Goal Actor {
           class=FortBTTask_GameMoveTo
           decorators [Distance Between {
             class=FortBTDecorator_DistanceBetween
             properties {
              flowAbortMode=Both
              nodeName=Distance Between
              operator=Greater
              blackboardKeyB {selectedKeyName=GoalActor}
              specifiedDistance {defaultValue=200}
             }
            }]
           properties {
            bDeimosFlavor=true
            bFinishMoveOnOverlap=false
            bStopAtGoal=true
            bUpdateMoveOnBlackboardKeyChanges=true
            nodeName=Fallback Move to Goal Actor
            acceptableRadius {defaultValue=250}
            bReachTestIncludesAgentRadius {defaultValue=false}
            bReachTestIncludesGoalRadius {defaultValue=false}
            blackboardKey {selectedKeyName=GoalActor}
            filterClass {defaultValue=&NavFilter_NPC_ExcludeWater_C}
            navFilterTag {tagName=None}
           }
          }
          Succeed {
           class=BTTask_FinishWithResult
           children []
           decorators [Can Actor be considered Airborne {
             class=FortBTDecorator_CanBeConsideredAirborne
             properties {
              bInverseCondition=false
              flowAbortMode=None
              nodeName=Can Actor be considered Airborne
              blackboardKey {bNoneIsAllowedValue=false, selectedKeyName=GoalActor}
             }
            }]
           properties {
            bIgnoreRestartSelf=false
            nodeName=Succeed
            result {defaultValue=0, enumType=/Script/AIModule.EBTNodeResult, key=None, nativeEnumTypeName=EBTNodeResult}
           }
           services []
          }
          Wait {class=BTTask_Wait, properties {waitTime {defaultValue=5}}}
         ]
         decorators [Loop {class=BTDecorator_Loop}]
         properties {
          children [
           {
            childComposite=&self:BTComposite_Selector_4
            childTask=null
            decorators [
             &self:BTDecorator_Blackboard_13
             &self:FortBTDecorator_DistanceBetween_1
            ]
           }
           {
            childComposite=null
            childTask=&self:FortBTTask_GameMoveTo_7
            decorators [&self:FortBTDecorator_DistanceBetween_0]
           }
           {
            childComposite=null
            childTask=&self:BTTask_FinishWithResult_0
            decorators [&self:FortBTDecorator_CanBeConsideredAirborne_3]
           }
           {childComposite=null, childTask=&self:BTTask_Wait_1, decorators []}
          ]
          services []
         }
         services []
        }
       ]
       decorators [Has Combat Token {
         class=FortBTDecorator_HasToken
         properties {
          bInverseCondition=false
          flowAbortMode=Both
          nodeName=Has Combat Token
          activationTagQuery {
           bCanEditDefaultValueType=false
           key=None
           defaultValue {
            _structType=/Script/GameplayTags.GameplayTagQuery
            autoDescription=""
            tokenStreamVersion=0
            userDescription=""
            queryTokenStream []
            tagDictionary []
           }
          }
          blackboardKey {bNoneIsAllowedValue=false, selectedKeyName=GoalActor}
          tokenQuery {
           bCanEditDefaultValueType=false
           key=None
           defaultValue {
            _structType=&GameplayTagQuery
            autoDescription=" ANY( Token.AI.BR.Melee )"
            tokenStreamVersion=0
            userDescription=""
            queryTokenStream [0,1,1,1,0]
            tagDictionary [{tagName=Token.AI.BR.Melee}]
           }
          }
         }
        }]
       properties {
        children [
         {
          childTask=&self:BTTask_Wait_12
          decorators [
           &self:FortBTDecorator_CanBeConsideredAirborne_2
           &self:BTDecorator_ForceSuccess_0
          ]
         }
         {
          childComposite=&self:BTComposite_Selector_3
          childTask=null
          decorators [&self:BTDecorator_Loop_7]
         }
        ]
        services [
         &self:FortAthenaBTService_AIEvaluator_0
         &self:FortAthenaBTService_ApplyGameplayTags_0
        ]
       }
       services [
        Token Position Evaluator {
         class=FortAthenaBTService_AIEvaluator
         properties {
          interval=0.5
          nodeName=Token Position Evaluator
          aIEvaluatorClass {defaultValue=&/EVAL_BuzzCool_Clone_MeleePosition.EVAL_BuzzCool_Clone_MeleePosition_C}
         }
        }
        Service Apply Gameplay Tags {
         class=FortAthenaBTService_ApplyGameplayTags
         properties {
          interval=0.5
          nodeName=Service Apply Gameplay Tags
          gameplayTagsToApply {defaultValue {gameplayTags [{tagName=BuzzCool.AI.Behaviors.Token.HasCombat}]}}
         }
        }
       ]
      }
      Selector {
       class=BTComposite_Selector
       children [
        GameplayAbility {
         class=FortBTTask_ExecuteGameplayAbility
         decorators [
          Has Gameplay Ability {
           class=FortBTDecorator_GameplayAbility_HasGameplayAbility
           properties {gameplayAbilityTagContainer {defaultValue {gameplayTags [{tagName=Ability.BuzzCool.Clone.Wander.Hostile}]}}}
          }
          Can Activate Gameplay Ability {
           class=FortBTDecorator_GameplayAbility_CanActivate
           properties {
            flowAbortMode=Both
            gameplayAbilityTagContainer {defaultValue {gameplayTags [{tagName=Ability.BuzzCool.Clone.Wander.Hostile}]}}
            targetValue {key=None}
           }
          }
         ]
         properties {nodeName=GameplayAbility, gameplayAbilityTagContainer {defaultValue {gameplayTags [{tagName=Ability.BuzzCool.Clone.Wander.Hostile}]}}}
        }
        GameplayAbility {
         class=FortBTTask_ExecuteGameplayAbility
         decorators [
          Has Gameplay Ability {
           class=FortBTDecorator_GameplayAbility_HasGameplayAbility
           properties {gameplayAbilityTagContainer {defaultValue {gameplayTags [{tagName=Ability.BuzzCool.Clone.Idle.Break}]}}}
          }
          Can Activate Gameplay Ability {
           class=FortBTDecorator_GameplayAbility_CanActivate
           properties {
            flowAbortMode=Both
            gameplayAbilityTagContainer {defaultValue {gameplayTags [{tagName=Ability.BuzzCool.Clone.Idle.Break}]}}
            targetValue {key=None}
           }
          }
         ]
         properties {nodeName=GameplayAbility, gameplayAbilityTagContainer {defaultValue {gameplayTags [{tagName=Ability.BuzzCool.Clone.Idle.Break}]}}}
        }
        Wait {class=BTTask_Wait, properties {randomDeviation {defaultValue=0.5}, waitTime {defaultValue=1}}}
       ]
       decorators [Loop {class=BTDecorator_Loop}]
       properties {
        children [
         {
          childComposite=null
          childTask=&self:FortBTTask_ExecuteGameplayAbility_1
          decorators [
           &self:FortBTDecorator_GameplayAbility_HasGameplayAbility_1
           &self:FortBTDecorator_GameplayAbility_CanActivate_1
          ]
         }
         {
          childComposite=null
          childTask=&self:FortBTTask_ExecuteGameplayAbility_2
          decorators [
           &self:FortBTDecorator_GameplayAbility_HasGameplayAbility_3
           &self:FortBTDecorator_GameplayAbility_CanActivate_3
          ]
         }
         {childComposite=null, childTask=&self:BTTask_Wait_2, decorators []}
        ]
        services []
       }
       services []
      }
     ]
     decorators [Is hostile goal? {
       class=BTDecorator_Blackboard
       properties {
        arithmeticOperation=Equal
        intValue=0
        nodeName=Is hostile goal?
        notifyObserver=ValueChange
        stringValue=""
        blackboardKey {selectedKeyName=GoalActorIsHostile}
       }
      }]
     properties {
      children [
       {
        childComposite=&self:BTComposite_Sequence_2
        childTask=null
        decorators [&self:FortBTDecorator_HasToken_1]
       }
       {
        childComposite=&self:BTComposite_Selector_5
        childTask=null
        decorators [&self:BTDecorator_Loop_2]
       }
      ]
      services [&self:FortAthenaBTService_TokenQuery_2]
     }
     services [Executing...Token Query {
       class=FortAthenaBTService_TokenQuery
       properties {
        bCallTickOnSearchStart=false
        bRestartTimerOnEachActivation=false
        interval=99999
        nodeName=Executing...Token Query
        randomDeviation=0
        targetActorName=GoalActor
        weaponKeyName=AIEvaluator_Global_DesiredWeapon
        activationTagQuery {
         bCanEditDefaultValueType=false
         key=None
         defaultValue {
          _structType=&GameplayTagQuery
          autoDescription=""
          tokenStreamVersion=0
          userDescription=""
          queryTokenStream []
          tagDictionary []
         }
        }
        becomeRelevantTokenAction {
         defaultValue=1
         enumType=/Script/FortniteAIServer.EFortAthenaBTServiceTokenQueryAction
         key=None
         nativeEnumTypeName=EFortAthenaBTServiceTokenQueryAction
        }
        ceaseRelevantTokenAction {
         defaultValue=2
         enumType=&EFortAthenaBTServiceTokenQueryAction
         key=None
         nativeEnumTypeName=EFortAthenaBTServiceTokenQueryAction
        }
        defaultTokenSystemTagQuery {
         bCanEditDefaultValueType=false
         key=None
         defaultValue {
          _structType=&GameplayTagQuery
          autoDescription=" ANY( Token.AI.BR.Melee )"
          tokenStreamVersion=0
          userDescription=""
          queryTokenStream [0,1,1,1,0]
          tagDictionary [{tagName=Token.AI.BR.Melee}]
         }
        }
        equippedItemTagAssociationDatas []
        targetChangeTokenAction {
         defaultValue=1
         enumType=&EFortAthenaBTServiceTokenQueryAction
         key=None
         nativeEnumTypeName=EFortAthenaBTServiceTokenQueryAction
        }
        tokenReleaseDelay {defaultValue=0, key=None}
        weaponChangeTokenAction {
         defaultValue=1
         enumType=&EFortAthenaBTServiceTokenQueryAction
         key=None
         nativeEnumTypeName=EFortAthenaBTServiceTokenQueryAction
        }
       }
      }]
    }
    Selector {
     class=BTComposite_Selector
     children [
      Sequence {
       class=BTComposite_Sequence
       children [
        Wait {class=BTTask_Wait, properties {randomDeviation {defaultValue=3}, waitTime {defaultValue=18}}}
        BTTask_BlueprintBase_ApplyGameplayEffect_C_0 {
         class=BTTask_BlueprintBase_ApplyGameplayEffect_C
         children []
         decorators []
         properties {
          bIgnoreRestartSelf=false
          bShowPropertyDetails=false
          customDescription=""
          gameplay Effect Class=/BuzzCool/NPCs/Abilities/GE_WildEstate_Location_Idle_Cooldown.GE_WildEstate_Location_Idle_Cooldown_C
          nodeName=""
          tickInterval {interval=-1}
         }
         services []
        }
       ]
       decorators [Gameplay Tag Query {
         class=BTDecorator_GameplayTagQuery
         properties {
          flowAbortMode=LowerPriority
          nodeName=Gameplay Tag Query
          gameplayTagQuery {
           autoDescription=" NONE( Ability.BuzzCool.Clone.Wander.Location.Idle.Cooldown )"
           queryTokenStream [0,1,3,1,0]
           tagDictionary [{tagName=Ability.BuzzCool.Clone.Wander.Location.Idle.Cooldown}]
          }
         }
        }]
       properties {children [
         {childTask=&self:BTTask_Wait_6}
         {childTask=&self:BTTask_BlueprintBase_ApplyGameplayEffect_C_0}
        ]}
      }
      Sequence {
       class=BTComposite_Sequence
       children [
        Find Spot Around Target Actor {
         class=BTTask_RunEQSQuery
         properties {
          nodeName=Find Spot Around Target Actor
          blackboardKey {selectedKeyName=WanderLocation}
          eQSRequest {queryTemplate=/BuzzCool/AI/EQS/EQS_BuzzCool_Clone_Wander_CloseToItem, runMode=RandomBest25Pct}
         }
        }
        Move towards Wander Loc {
         class=FortBTTask_GameMoveTo
         properties {
          bDeimosFlavor=true
          bFinishMoveOnOverlap=false
          bStopAtGoal=true
          bUpdateMoveOnBlackboardKeyChanges=true
          nodeName=Move towards Wander Loc
          pushBumpedPawnClass=/BuzzCool/NPCs/BuzzCool_Clones/Pawn/Pawn_WildEstate_BuzzCool.Pawn_WildEstate_BuzzCool_C
          acceptableRadius {defaultValue=300}
          bReachTestIncludesAgentRadius {defaultValue=false}
          bReachTestIncludesGoalRadius {defaultValue=false}
          blackboardKey {selectedKeyName=WanderLocation}
          filterClass {defaultValue=&NavFilter_NPC_ExcludeWater_C}
          navFilterTag {tagName=None}
          observedBlackboardValueTolerance {defaultValue=1}
         }
        }
       ]
       decorators [BTDecorator_CanPathAroundGoal_C_0 {
         class=BTDecorator_CanPathAroundGoal_C
         properties {
          bCheckConditionOnlyBlackBoardChanges=false
          bInverseCondition=false
          bShowPropertyDetails=true
          customDescription=""
          flowAbortMode=None
          nodeName=""
          sF_MaxPathingAllowed {registryType=None, value=1, curve {curveTable=/BuzzCool/DataTables/BuzzCool_GameData, rowName=Default.BuzzCool.Clone.Wander.Item.MaxPathing}}
         }
        }]
       properties {
        children [
         {childTask=&self:BTTask_RunEQSQuery_0}
         {childTask=&self:FortBTTask_GameMoveTo_1}
        ]
        services [&self:FortAthenaBTService_ApplyGameplayTags_1]
       }
       services [Service Apply Gameplay Tags {
         class=FortAthenaBTService_ApplyGameplayTags
         properties {
          interval=0.5
          nodeName=Service Apply Gameplay Tags
          bReplicateTags {defaultValue=true}
          gameplayTagsToApply {defaultValue {gameplayTags [{tagName=Ability.BuzzCool.Clone.Wander.Location.Skipping}]}}
         }
        }]
      }
      Wait {class=BTTask_Wait, properties {randomDeviation {defaultValue=1}, waitTime {defaultValue=3}}}
     ]
     decorators [Loop {class=BTDecorator_Loop}]
     properties {
      children [
       {
        childComposite=&self:BTComposite_Sequence_3
        childTask=null
        decorators [&self:BTDecorator_GameplayTagQuery_0]
       }
       {
        childComposite=&self:BTComposite_Sequence_0
        childTask=null
        decorators [&self:BTDecorator_CanPathAroundGoal_C_0]
       }
       {childComposite=null, childTask=&self:BTTask_Wait_3, decorators []}
      ]
      services []
     }
     services []
    }
   ]
   decorators [Has Goal Actor? {
     class=BTDecorator_Blackboard
     properties {
      arithmeticOperation=Equal
      intValue=0
      nodeName=Has Goal Actor?
      notifyObserver=ValueChange
      stringValue=""
      blackboardKey {selectedKeyName=GoalActor}
     }
    }]
   properties {
    children [
     {
      childComposite=&self:BTComposite_Selector_1
      childTask=null
      decorators [&self:BTDecorator_Blackboard_0]
     }
     {
      childComposite=&self:BTComposite_Selector_6
      childTask=null
      decorators [&self:BTDecorator_Loop_4]
     }
    ]
    services []
   }
   services []
  }
  Selector {
   class=BTComposite_Selector
   children [
    GameplayAbility {
     class=FortBTTask_ExecuteGameplayAbility
     decorators [
      Has Gameplay Ability {
       class=FortBTDecorator_GameplayAbility_HasGameplayAbility
       properties {gameplayAbilityTagContainer {defaultValue {gameplayTags [{tagName=Ability.BuzzCool.Clone.Wander.Free}]}}}
      }
      Can Activate Gameplay Ability {
       class=FortBTDecorator_GameplayAbility_CanActivate
       properties {
        flowAbortMode=Both
        gameplayAbilityTagContainer {defaultValue {gameplayTags [{tagName=Ability.BuzzCool.Clone.Wander.Free}]}}
        targetValue {key=None}
       }
      }
     ]
     properties {nodeName=GameplayAbility, gameplayAbilityTagContainer {defaultValue {gameplayTags [{tagName=Ability.BuzzCool.Clone.Wander.Free}]}}}
    }
    Wait {class=BTTask_Wait, properties {waitTime {defaultValue=9999}}}
   ]
   decorators [Loop {class=BTDecorator_Loop}]
   properties {
    children [
     {
      childComposite=null
      childTask=&self:FortBTTask_ExecuteGameplayAbility_0
      decorators [
       &self:FortBTDecorator_GameplayAbility_HasGameplayAbility_0
       &self:FortBTDecorator_GameplayAbility_CanActivate_0
      ]
     }
     {childComposite=null, childTask=&self:BTTask_Wait_0, decorators []}
    ]
    services []
   }
   services []
  }
 ]
 decorators []
 properties {
  children [
   {
    childComposite=&self:BTComposite_Selector_0
    childTask=null
    decorators [&self:BTDecorator_Blackboard_6]
   }
   {
    childComposite=&self:BTComposite_Selector_13
    childTask=null
    decorators [&self:BTDecorator_Loop_3]
   }
  ]
  services []
 }
 services []
}
